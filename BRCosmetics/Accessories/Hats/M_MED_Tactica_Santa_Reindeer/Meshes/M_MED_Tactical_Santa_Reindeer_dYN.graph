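# Physics asset for the Tactical Santa Reindeer hat cosmetic, defining bodies and joints for its dangling hat, bow and bell.
# A four-segment hat chain hangs from the head with varying swing limits and tight twist.
# The bow base joins spine_05, branching into two loops, two two-segment strings and a bell with its ringer.
# Large head and upper-spine capsules form the body volumes, linked by free angular constraints.
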PhysicsAsset /BRCosmetics/Accessories/Hats/M_MED_Tactica_Santa_Reindeer/Meshes/M_MED_Tactical_Santa_Reindeer_dYN
bodies {
 dyn_bell [{shapeType=Capsule, length=8.3834, radius=2.9269, center=<0.4338 0 -0.808>, rotation=<rot 90 0 0>}]
 dyn_bow_base [{shapeType=Capsule, length=2, radius=2, center=<0.2496 0 0>, rotation=<rot 90 0 0>}]
 dyn_bow_loop_l [{shapeType=Capsule, length=2, radius=2, center=<0 -3.14 0>, rotation=<rot 0 0 89.9994>}]
 dyn_bow_loop_r [{shapeType=Capsule, length=2, radius=2, center=<0 3.0617 0>, rotation=<rot 0 0 89.9994>}]
 dyn_bow_string_l_01 [{shapeType=Box, extentX=5.4377, extentY=2.5, extentZ=1, center=<0.3128 0 0>}]
 dyn_bow_string_l_02 [
  {shapeType=Box, extentX=2.5707, extentY=3, extentZ=1, center=<1.1539 0.0001 0.2001>, rotation=<rot 10 0 0>}
  {shapeType=Box, extentX=3.0554, extentY=3.5376, extentZ=1.6906, center=<2.6795 -0.0003 0.7036>, rotation=<rot 19.9999 0 0>}
 ]
 dyn_bow_string_r_01 [{shapeType=Box, extentX=4.9746, extentY=2.5, extentZ=1, center=<0.5005 0 -0.0003>}]
 dyn_bow_string_r_02 [
  {shapeType=Box, extentX=2.7969, extentY=3, extentZ=1, center=<1.1891 0 0.0001>}
  {shapeType=Box, extentX=3.7728, extentY=4.9167, extentZ=1.7057, center=<3.3235 0.0001 0.5516>, rotation=<rot 14.9999 0 0>}
 ]
 dyn_hat_01 [{shapeType=Capsule, length=2, radius=2, center=<0.2 0 0>, rotation=<rot 90 0 0>}]
 dyn_hat_02 [{shapeType=Capsule, length=2, radius=2, center=<0.1 0 0>, rotation=<rot 90 0 0>}]
 dyn_hat_03 [{shapeType=Capsule, length=2, radius=2, center=<0.1 0 0>, rotation=<rot 90 0 0>}]
 dyn_hat_04 [{shapeType=Capsule, length=4.2372, radius=2.3245, center=<0.1 0 0>, rotation=<rot 90 0 0>}]
 dyn_ringer [{shapeType=Capsule, length=5.0058, radius=0.547, center=<2.3388 0 -0.0002>, rotation=<rot 90 0 0>}]
 head [{shapeType=Capsule, length=13.1372, radius=11.5841, center=<0.7107 -1.9884 0.4339>, rotation=<rot 90 0 0>}]
 spine_03 [{shapeType=Capsule, length=13.8686, radius=10.7993, center=<6.8866 -7.4204 -0.0002>, rotation=<rot 0 -90 0>}]
 spine_04 [{shapeType=Capsule, length=13.8686, radius=10.7993, center=<6.5701 -7.4162 -0.0002>, rotation=<rot 0 -90 0>}]
 spine_05 [{shapeType=Capsule, length=21.8807, radius=22.7528, center=<-1.3469 -7.5106 0>, rotation=<rot 0 -90 0>}]
}
constraints [
 {bone1Name=dyn_hat_03, bone2Name=dyn_hat_02, swing1LimitDegrees=75, swing2LimitDegrees=75, twistLimitDegrees=8}
 {bone1Name=dyn_bow_string_l_01, bone2Name=dyn_bow_base, swing1LimitDegrees=10, swing2LimitDegrees=45, twistLimitDegrees=2}
 {bone1Name=dyn_bow_string_l_02, bone2Name=dyn_bow_string_l_01, swing1LimitDegrees=10, swing2LimitDegrees=45, twistLimitDegrees=2}
 {bone1Name=dyn_bow_string_r_01, bone2Name=dyn_bow_base, swing1LimitDegrees=10, swing2LimitDegrees=45, twistLimitDegrees=2}
 {bone1Name=dyn_bow_string_r_02, bone2Name=dyn_bow_string_r_01, swing1LimitDegrees=10, swing2LimitDegrees=45, twistLimitDegrees=2}
 {bone1Name=dyn_bow_loop_r, bone2Name=dyn_bow_base, swing1LimitDegrees=25, swing2LimitDegrees=2, twistLimitDegrees=25}
 {bone1Name=dyn_bow_loop_l, bone2Name=dyn_bow_base, swing1LimitDegrees=25, swing2LimitDegrees=2, twistLimitDegrees=25}
 {bone1Name=dyn_bell, bone2Name=dyn_bow_base, swing1LimitDegrees=45, swing2LimitDegrees=45, twistLimitDegrees=15}
 {bone1Name=dyn_hat_01, bone2Name=head, swing1LimitDegrees=45, swing2LimitDegrees=45, twistLimitDegrees=15}
 {bone1Name=dyn_bow_base, bone2Name=spine_05, swing1LimitDegrees=25, swing2LimitDegrees=25, twistLimitDegrees=10}
 {
  bone1Name=head
  bone2Name=spine_05
  swing1LimitDegrees=45
  swing1Motion=Free
  swing2LimitDegrees=45
  swing2Motion=Free
  twistLimitDegrees=45
  twistMotion=Free
 }
 {bone1Name=dyn_hat_02, bone2Name=dyn_hat_01, swing1LimitDegrees=60, swing2LimitDegrees=90, twistLimitDegrees=5}
 {bone1Name=dyn_hat_04, bone2Name=dyn_hat_03, swing1LimitDegrees=20, swing2LimitDegrees=20, twistLimitDegrees=5}
 {bone1Name=dyn_ringer, bone2Name=dyn_bell, swing1LimitDegrees=15, swing2LimitDegrees=15, twistLimitDegrees=5}
 {
  bone1Name=spine_05
  bone2Name=spine_04
  swing1LimitDegrees=45
  swing1Motion=Free
  swing2LimitDegrees=45
  swing2Motion=Free
  twistLimitDegrees=45
  twistMotion=Free
 }
 {
  bone1Name=spine_04
  bone2Name=spine_03
  swing1LimitDegrees=45
  swing1Motion=Free
  swing2LimitDegrees=45
  swing2Motion=Free
  twistLimitDegrees=45
  twistMotion=Free
 }
]
