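# Animation Blueprint for the tactical Santa reindeer hat cosmetic, adding secondary physics to a pose copied from the attached parent mesh.
# A RigidBody simulation relative to spine_05 feeds an angular-spring AnimDynamics on dyn_ringer, both using gravity updated each frame by a deprecated override function.
# The cached result outputs directly, or with the head bone scaled near zero when the playlist uses custom character parts, a flag read at initialization.
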AnimBlueprint /BRCosmetics/Accessories/Hats/M_MED_Tactica_Santa_Reindeer/Meshes/M_MED_Tactical_Santa_Reindeer_AnimBP
target_skeleton=&/M_MED_Tactical_Santa_Reindeer_Skeleton
graphs [
 AnimGraph {nodes [
   n1 = "Transform (Modify) Bone - Bone: head"!(ComponentPose=Local To Component(LocalPose=Use cached pose 'PreScale Cache'()), Alpha=1) -> (Pose) {scaleMode=BMM_Replace, alphaBoolBlend {blendOption=Linear}, boneToModify {boneName=head}, scale=<0.01 0.01 0.01>}
   Output Pose(Result=Blend Poses by bool(
     BlendPose_0=Component To Local(ComponentPose=@n1)
     BlendPose_1=Use cached pose 'PreScale Cache'()
     BlendTime_0=0
     BlendTime_1=0
     bActiveValue=Get GetPlayerListUsesCustomCharacterParts() -> (GetPlayerListUsesCustomCharacterParts:bool)
    ) {blendType=Linear}) {layerGroup=None}
   n2 = Get GravityOverride()
   n3 = RigidBody!(
    ComponentPose=Local To Component(LocalPose=Copy Pose From Mesh() {bUseAttachedParent=true})
    Alpha
    OverrideWorldGravity:Vector=@n2
    ExternalForce
   ) -> (Pose) {
    bOverrideWorldGravity=true
    overridePhysicsAsset=&/M_MED_Tactical_Santa_Reindeer_dYN
    simulationSpace=BaseBoneSpace
    alphaBoolBlend {blendOption=Linear}
    baseBoneRef {boneName=spine_05}
    componentAppliedLinearAccClamp=<2000 2000 2000>
    componentLinearAccScale=<3 3 3>
    componentLinearVelScale=<3 3 3>
    overrideWorldGravity {z=-980}
   }
   n4 = Anim Dynamics - Bone: dyn_ringer!(ComponentPose=@n3, Alpha=1, GravityOverride:Vector=@n2, ExternalForce="0,0,0") -> (Pose) {
    angularDampingOverride=1
    bAngularSpring=true
    bOverrideAngularDamping=true
    bUseGravityOverride=true
    gravityScale=0.35
    simulationSpace=RootRelative
    alphaBoolBlend {blendOption=Linear}
    boundBone {boneName=dyn_ringer}
    componentAppliedLinearAccClamp=<3000 3000 3000>
    componentLinearAccScale=<1 1 1>
    componentLinearVelScale=<1 1 1>
    gravityOverride {z=-980}
    physicsBodyDefinitions [{
      boundBone {boneName=dyn_ringer}
      boxExtents=<15 20 3>
      constraintSetup {angularLimitsMax=<0 0 23>, angularLimitsMin=<0 -0.8 -7>, angularTarget=<1 0 0>}
      localJointOffset=<7.5985 0.845 -0.8244>
     }]
   }
   Save cached pose 'PreScale Cache'(Pose=Component To Local(ComponentPose=@n4)) {cacheName=PreScale Cache}
  ]}
 EventGraph {nodes [
   Event BlueprintUpdateAnimation()
   > n1 = Gravity Override State Function DEPRECATED(CharacterPartAnimInstance=Self-Reference(), GameTimeInSeconds=GetGameTimeInSeconds(), GravityOverrideParams=Get Gravity Override Params())
   > Set GravityOverride(GravityOverride=@n1)
   Event BlueprintInitializeAnimation()
   > Set GetPlayerListUsesCustomCharacterParts(GetPlayerListUsesCustomCharacterParts:bool=GetPlaylistUsesCustomCharacterParts!() -> (ReturnValue:bool)) -> (Output_Get:bool)
   TryGetPawnOwner()
   Set jaw_rot(jaw_rot:Rotator=GetSocketRotation!(self:Scene Component Object Reference=GetOwnerSkeletalMeshComponent(), InSocketName:Name=C_jaw) -> (ReturnValue:Rotator)) -> (Output_Get:Rotator)
  ]}
]
variables [
 GetPlayerListUsesCustomCharacterParts {type=boolean}
 Gravity Override Params {type=GravityOverrideParamsStruct}
 GravityOverride {type=Vector}
 jaw_rot {type=Rotator}
]
